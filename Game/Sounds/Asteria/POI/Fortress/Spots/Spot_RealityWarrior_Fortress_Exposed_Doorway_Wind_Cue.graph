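# Sound cue for an ambient wind spot, named for an exposed doorway at the Reality Warrior Fortress POI in Asteria.
# A Random node picks equally between two volume envelopes over the same Winterfest exterior wind wave: a plain fade in and out, or a delayed gust swell.
# Two chained, indefinitely looping Envelopers above it add slight pitch and volume drift.
# It uses tile-ambient wind attenuation and the indoor ambient sound class, and keeps playing when silent.
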SoundCue /Game/Sounds/Asteria/POI/Fortress/Spots/Spot_RealityWarrior_Fortress_Exposed_Doorway_Wind_Cue
root=Enveloper_3
nodes [
 {
  id=Enveloper_0
  children [WavePlayer_0]
  properties {pitchCurve {editorCurveData {keys [{time=0}]}}, volumeCurve {editorCurveData {keys [<at 0 0>,<at 2 1>,<at 8 1>,<at 10 0>]}}}
 }
 {
  id=Enveloper_1
  children [WavePlayer_1]
  properties {
   pitchCurve {editorCurveData {keys [{time=0}]}}
   volumeCurve {editorCurveData {keys [
      <at 0 0>
      <at 2 0>
      {arriveTangent=0.3556, interpMode=RCIM_Cubic, leaveTangent=0.3556, time=4, value=1}
      <at 6.2178 1.5 RCIM_Cubic>
      {arriveTangent=-0.9232, interpMode=RCIM_Cubic, leaveTangent=-0.9232, tangentMode=RCTM_User, time=8, value=1}
      <at 10 0>
     ]}}
  }
 }
 {
  id=Enveloper_2
  children [Random_0]
  properties {
   bLoop=true
   bLoopIndefinitely=true
   loopEnd=5
   pitchCurve {editorCurveData {keys [{arriveTangent=-0.0013, interpMode=RCIM_Cubic, time=0}, <at 15 1.05 RCIM_Cubic>, {interpMode=RCIM_Cubic, leaveTangent=0.0013, time=30}]}}
   volumeCurve {editorCurveData {keys [<at 0 1>]}}
  }
 }
 {
  id=Enveloper_3
  children [Enveloper_2]
  properties {
   bLoop=true
   bLoopIndefinitely=true
   loopEnd=5
   pitchCurve {editorCurveData {keys [{time=0}]}}
   volumeCurve {editorCurveData {keys [
      {arriveTangent=-0.01, interpMode=RCIM_Cubic, time=0, value=1}
      <at 10 1.05 RCIM_Cubic>
      {interpMode=RCIM_Cubic, leaveTangent=0.01, time=20, value=1}
     ]}}
  }
 }
 {id=Random_0, children [Enveloper_0,Enveloper_1], properties {weights [1,1]}}
 {id=WavePlayer_0, properties {soundWaveAssetPtr=/Game/Athena/Sounds/Events/Winterfest/Winterfest_Exterior_Wind_Loop}}
 {id=WavePlayer_1, properties {soundWaveAssetPtr=&Winterfest_Exterior_Wind_Loop}}
]
properties {
 attenuationSettings=/Game/Sounds/Attenuation/Fort_TileAmbient_OneShot_Wind_Att
 maxDistance=8000
 soundClassObject=&Fort_Ambient_Indoors_Athena
 virtualizationMode=PlayWhenSilent
 volumeMultiplier=0.501
}
