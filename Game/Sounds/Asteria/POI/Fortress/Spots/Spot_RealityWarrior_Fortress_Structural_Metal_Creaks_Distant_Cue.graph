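# Sound cue for distant structural metal creak one-shots at the Reality Warrior Fortress point of interest, classed as indoor ambient one-shots.
# A Random node picks among six equally weighted distant blimp hull metal impact waves reused from Artemis sounds, feeding a Modulator that varies pitch.
# It plays through a less-positional tile ambient attenuation with overridden falloff and stereo spread, with bus and submix sends disabled.
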SoundCue /Game/Sounds/Asteria/POI/Fortress/Spots/Spot_RealityWarrior_Fortress_Structural_Metal_Creaks_Distant_Cue
root=Modulator_0
nodes [
 {id=Modulator_0, children [Random_0], properties {pitchMin=0.9}}
 {id=Random_0, children [WavePlayer_0,WavePlayer_1,WavePlayer_2,WavePlayer_3,WavePlayer_4,WavePlayer_5], properties {weights [1,1,1,1,1,1]}}
 {id=WavePlayer_0, properties {soundWaveAssetPtr=/Game/Sounds/Artemis/POI/Blimps/Spots/Spots_Blimp_Metal_Hull_Impacts_Distant_01}}
 {id=WavePlayer_1, properties {soundWaveAssetPtr=/Game/Sounds/Artemis/POI/Blimps/Spots/Spots_Blimp_Metal_Hull_Impacts_Distant_02}}
 {id=WavePlayer_2, properties {soundWaveAssetPtr=/Game/Sounds/Artemis/POI/Blimps/Spots/Spots_Blimp_Metal_Hull_Impacts_Distant_03}}
 {id=WavePlayer_3, properties {soundWaveAssetPtr=/Game/Sounds/Artemis/POI/Blimps/Spots/Spots_Blimp_Metal_Hull_Impacts_Distant_04}}
 {id=WavePlayer_4, properties {soundWaveAssetPtr=/Game/Sounds/Artemis/POI/Blimps/Spots/Spots_Blimp_Metal_Hull_Impacts_Distant_05}}
 {id=WavePlayer_5, properties {soundWaveAssetPtr=/Game/Sounds/Artemis/POI/Blimps/Spots/Spots_Blimp_Metal_Hull_Impacts_Distant_06}}
]
properties {
 attenuationSettings=/Game/Sounds/Attenuation/Fort_TileAmbient_OneShot_LessPositional_Log_Att
 bEnableBusSends=false
 bEnableSubmixSends=false
 maxDistance=6200
 soundClassObject=/Game/Sounds/FortSoundClassesAndModes/Fort_Ambient_OneShots_Indoors
 volumeMultiplier=1.504
 attenuationOverrides {falloffDistance=6000, stereoSpread=10000, attenuationShapeExtents {x=2000}}
}
