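# Sound cue for a quiet ice-crystal air ambient spot one-shot inside the Asteria Fortress POI.
# A Modulator feeds a quality-level switch: the first two levels randomly pick from five wave variations, the last two from only two.
# It plays through an indoor ambient sound class with less-positional tile one-shot attenuation, yet also sends to the exterior ambient source bus.
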SoundCue /Game/Sounds/Asteria/POI/Fortress/Spots/Spot_Reality_Fortress_Ice_Crystal_Air_Interior_Cue
root=Modulator_0
nodes [
 {id=Modulator_0, children [QualityLevel_0]}
 {id=QualityLevel_0, children [Random_2,Random_2,Random_3,Random_3]}
 {id=Random_2, children [WavePlayer_18, WavePlayer_19, WavePlayer_20, WavePlayer_21, WavePlayer_22], properties {weights [1,1,1,1,1]}}
 {id=Random_3, children [WavePlayer_25, WavePlayer_23], properties {weights [1,1]}}
 {id=WavePlayer_18, properties {soundWaveAssetPtr=&/Spot_Fortress_Interior_Ice_Crystal_Air_5}}
 {id=WavePlayer_19, properties {soundWaveAssetPtr=&/Spot_Fortress_Interior_Ice_Crystal_Air_1}}
 {id=WavePlayer_20, properties {soundWaveAssetPtr=&/Spot_Fortress_Interior_Ice_Crystal_Air_2}}
 {id=WavePlayer_21, properties {soundWaveAssetPtr=&/Spot_Fortress_Interior_Ice_Crystal_Air_3}}
 {id=WavePlayer_22, properties {soundWaveAssetPtr=&/Spot_Fortress_Interior_Ice_Crystal_Air_4}}
 {id=WavePlayer_23, properties {soundWaveAssetPtr=&Spot_Fortress_Interior_Ice_Crystal_Air_1}}
 {id=WavePlayer_25, properties {soundWaveAssetPtr=&Spot_Fortress_Interior_Ice_Crystal_Air_5}}
]
properties {
 attenuationSettings=/Game/Sounds/Attenuation/Fort_TileAmbient_OneShot_LessPositional_Log_Att
 maxDistance=6200
 soundClassObject=&Fort_Ambient_Indoors_Athena
 volumeMultiplier=0.1
 busSends [{audioBus=null, minSendLevel=0, sendLevel=1, soundSourceBus=&AmbientAudio_Exterior_SourceBus}]
}
