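# Sound cue for spot one-shot ambience of shifting sounds inside the Asteria Fortress POI, played on the indoor ambient one-shot sound class.
# A quality-level switch sends its first two tiers to an equally weighted random pick of fourteen variations and its last two to a four-wave subset, trimming variety by quality.
# A root modulator wraps the switch, with reduced cue volume and less-positional tile-ambient attenuation.
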SoundCue /Game/Sounds/Asteria/POI/Fortress/Spots/Spot_Reality_Fortress_Shifting_Interior_Cue
root=Modulator_0
nodes [
 {id=Modulator_0, children [QualityLevel_0]}
 {id=QualityLevel_0, children [Random_0,Random_0,Random_1,Random_1]}
 {
  id=Random_0
  children [
   WavePlayer_0
   WavePlayer_1
   WavePlayer_2
   WavePlayer_3
   WavePlayer_4
   WavePlayer_5
   WavePlayer_6
   WavePlayer_7
   WavePlayer_8
   WavePlayer_9
   WavePlayer_10
   WavePlayer_11
   WavePlayer_12
   WavePlayer_13
  ]
  properties {weights [1,1,1,1,1,1,1,1,1,1,1,1,1,1]}
 }
 {id=Random_1, children [WavePlayer_17, WavePlayer_16, WavePlayer_15, WavePlayer_14], properties {weights [1,1,1,1]}}
 {id=WavePlayer_0, properties {soundWaveAssetPtr=&/Spot_Fortress_Interior_Shifting_14}}
 {id=WavePlayer_1, properties {soundWaveAssetPtr=&/Spot_Fortress_Interior_Shifting_01}}
 {id=WavePlayer_10, properties {soundWaveAssetPtr=&/Spot_Fortress_Interior_Shifting_10}}
 {id=WavePlayer_11, properties {soundWaveAssetPtr=&/Spot_Fortress_Interior_Shifting_11}}
 {id=WavePlayer_12, properties {soundWaveAssetPtr=&/Spot_Fortress_Interior_Shifting_12}}
 {id=WavePlayer_13, properties {soundWaveAssetPtr=&/Spot_Fortress_Interior_Shifting_13}}
 {id=WavePlayer_14, properties {soundWaveAssetPtr=&Spot_Fortress_Interior_Shifting_14}}
 {id=WavePlayer_15, properties {soundWaveAssetPtr=&Spot_Fortress_Interior_Shifting_12}}
 {id=WavePlayer_16, properties {soundWaveAssetPtr=&/Spot_Fortress_Interior_Shifting_07}}
 {id=WavePlayer_17, properties {soundWaveAssetPtr=&/Spot_Fortress_Interior_Shifting_04}}
 {id=WavePlayer_2, properties {soundWaveAssetPtr=&/Spot_Fortress_Interior_Shifting_02}}
 {id=WavePlayer_3, properties {soundWaveAssetPtr=&/Spot_Fortress_Interior_Shifting_03}}
 {id=WavePlayer_4, properties {soundWaveAssetPtr=&Spot_Fortress_Interior_Shifting_04}}
 {id=WavePlayer_5, properties {soundWaveAssetPtr=&/Spot_Fortress_Interior_Shifting_05}}
 {id=WavePlayer_6, properties {soundWaveAssetPtr=&/Spot_Fortress_Interior_Shifting_06}}
 {id=WavePlayer_7, properties {soundWaveAssetPtr=&Spot_Fortress_Interior_Shifting_07}}
 {id=WavePlayer_8, properties {soundWaveAssetPtr=&/Spot_Fortress_Interior_Shifting_08}}
 {id=WavePlayer_9, properties {soundWaveAssetPtr=&/Spot_Fortress_Interior_Shifting_09}}
]
properties {
 attenuationSettings=/Game/Sounds/Attenuation/Fort_TileAmbient_OneShot_LessPositional_Log_Att
 maxDistance=6200
 soundClassObject=/Game/Sounds/FortSoundClassesAndModes/Fort_Ambient_OneShots_Indoors
 volumeMultiplier=0.501
}
